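# Translucent UI material for the season 8 DragonNinja loading screen that transforms character art from a Base texture to an Incoming one.
# The Switch parameter drives a soft smoothstep threshold on a reveal mask's red channel, blending color and alpha.
# Switch also positions a band in the green channel that confines a time-animated lightning flipbook.
# Thresholded yellow lightning extends the opacity, with the lightning texture's blue channel controlling where it covers the character.
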Material /BRCosmetics/2dAssets/Loadingscreens/season8/DragonNinjaLoading/M_UI_DragonNinjaLoading_Character_Base
graph {
 properties {blendMode=BLEND_Translucent, materialDomain=MD_UI, pixelDepthOffsetMode=PDOM_Legacy, usedShadingModels=Default Lit}
 nodes [
  n1 = TextureSampleParameter2D() {
   parameterName=Base
   samplerSource=SSM_Clamp_WorldGroupSettings
   texture=&/T_UI_DragonNinja_p1
  }
  n2 = TextureSampleParameter2D() {
   parameterName=Incoming
   samplerSource=SSM_Clamp_WorldGroupSettings
   texture=&/T_UI_DragonNinja_p2
  }
  n3 = TextureSample() {
   samplerType=SAMPLERTYPE_Masks
   texture=&/T_UI_DragonNinjaLoading_Reveal_Channel
  }
  n4 = Subtract(A=ScalarParameter() {defaultValue=1, parameterName=Switch}, B=ScalarParameter() {parameterName=BodyRevealOffset})
  n5 = MaterialFunctionCall(Alpha=@n3.R, Min=@n4, Max=Add(A=@n4, B=ScalarParameter() {defaultValue=0.2, parameterName=RevealMaskSoftness})) {materialFunction=&SmoothStep}
  n6 = Clamp(None=LinearInterpolate(A=@n1.A, B=@n2.A, Alpha=@n5.Result))
  n7 = Multiply(A=Time(), B=Constant() {r=10})
  n8 = Constant() {r=1}
  n9 = Constant() {r=3}
  n10 = TextureObject() {texture=&/T_UI_DragonNinjaLoading_Lightning}
  n11 = TextureCoordinate()
  n12 = TextureSample() {
   samplerSource=SSM_Clamp_WorldGroupSettings
   samplerType=SAMPLERTYPE_Masks
   texture=&T_UI_DragonNinjaLoading_Reveal_Channel
  }
  n13 = ScalarParameter() {defaultValue=1, parameterName=Switch}
  n14 = MaterialFunctionCall(Alpha=@n12.G, Min=@n13, Max=Add(A=@n13, B=ScalarParameter() {defaultValue=0.2, parameterName=LightningMaskSoftness})) {materialFunction=&SmoothStep}
  n15 = TextureSample() {
   samplerSource=SSM_Clamp_WorldGroupSettings
   samplerType=SAMPLERTYPE_Masks
   texture=&T_UI_DragonNinjaLoading_Reveal_Channel
  }
  n16 = ScalarParameter() {defaultValue=1, parameterName=Switch}
  n17 = ScalarParameter() {defaultValue=0.2, parameterName=LightningMaskLength}
  n18 = OneMinus(None=MaterialFunctionCall(Alpha=@n15.G, Min=@n16, Max=Subtract(A=@n16, B=@n17)) {materialFunction=&SmoothStep}.Result)
  n19 = ScalarParameter() {defaultValue=0.1, parameterName=LightningStep}
  n20 = MaterialFunctionCall(
   Alpha=Multiply(
    A=MaterialFunctionCall("Animation  Phase (0-1)"=@n7, Number of Rows=@n8, Number of Columns=@n9, Texture=@n10, UVs=@n11) {materialFunction=&FlipBook}.Result
    B=Multiply(A=@n14.Result, B=@n18)
   )
   Min=@n19
   Max=Add(A=@n19, B=ScalarParameter() {defaultValue=0.08, parameterName=EdgeSmooth})
  ) {materialFunction=&SmoothStep}
  Clamp(None=Max(A=@n6, B=@n20.Result))
  n21 = Constant3Vector() {constant=<rgba 1 0.99 0 0>}
  n22 = LinearInterpolate(A=LinearInterpolate(A=@n21, B=@n1.RGB, Alpha=@n1.RGB), B=@n2.RGB, Alpha=@n5.Result)
  n23 = TextureSample() {texture=&T_UI_DragonNinjaLoading_Lightning}
  LinearInterpolate(A=LinearInterpolate(A=@n22, B=@n21, Alpha=@n20.Result), B=@n22, Alpha=Multiply(A=@n6, B=OneMinus(None=@n23.B)))
  ScalarParameter() {defaultValue=1, parameterName=Switch}
 ]
}
