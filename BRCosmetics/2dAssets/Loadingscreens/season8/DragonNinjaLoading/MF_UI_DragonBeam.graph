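# Material function that draws a soft-edged, angled beam mask in screen space for the season 8 Dragon Ninja loading screen.
# It rotates aspect-corrected screen UVs around the bottom-center point twice, using the packed Rotation/Width/Edge input, and smoothsteps the horizontal axis of each.
# The second edge is offset by width and rotated back by a third of it, so the sides are not parallel.
# Subtracting and clamping the two edges yields the mask, with EdgeSoftness setting the falloff.
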MaterialFunction /BRCosmetics/2dAssets/Loadingscreens/season8/DragonNinjaLoading/MF_UI_DragonBeam
graph {nodes [
  n1 = MaterialFunctionCall() {materialFunction=&ScreenAlignedUVs}
  n2 = Constant2Vector() {g=1, r=0.5}
  n3 = MaterialFunctionCall(Float4=FunctionInput(Preview=Constant4Vector() {constant=<rgba 0.0571 -0.2762 0.1048 0>}.RGBA) {bUsePreviewValueAsDefault=true, description=Rotation/Width/Edge, inputType=FunctionInput_Vector4}) {materialFunction=&BreakOutFloat4Components}
  n4 = MaterialFunctionCall(
   Alpha=ComponentMask(None=MaterialFunctionCall(UVs=@n1."X Scale to Ratio, Y 100%", Rotation Center=@n2, "Rotation Angle (0-1)"=@n3.R) {materialFunction=&CustomRotator}.Rotated Values) {r=true}
   Min=@n3.B
   Max=Add(A=@n3.B, B=ScalarParameter() {defaultValue=0.01, parameterName=EdgeSoftness})
  ) {materialFunction=&SmoothStep}
  n5 = Subtract(A=@n3.R, B=Divide(A=@n3.G) {constB=3})
  n6 = Add(A=@n3.B, B=@n3.G)
  n7 = MaterialFunctionCall(
   Alpha=ComponentMask(None=MaterialFunctionCall(UVs=@n1."X Scale to Ratio, Y 100%", Rotation Center=@n2, "Rotation Angle (0-1)"=@n5) {materialFunction=&CustomRotator}.Rotated Values) {r=true}
   Min=@n6
   Max=Add(A=@n6, B=ScalarParameter() {defaultValue=0.01, parameterName=EdgeSoftness})
  ) {materialFunction=&SmoothStep}
  FunctionOutput(None=Clamp(None=Subtract(A=@n4.Result, B=@n7.Result)))
 ]}
