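# Material function for the Dragon Ninja loading screen that outputs the mask of one spinning shuriken repeatedly flying horizontally across the screen.
# Aspect-corrected screen UVs are placed and scaled by the Size and Offset inputs, then slid by a looping Time phase whose speed and travel also depend on Size.
# The same phase drives CustomRotator for the spin.
# The clamped texture's red channel is smoothstepped by Edge parameters and multiplied by Alpha.
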MaterialFunction /BRCosmetics/2dAssets/Loadingscreens/season8/DragonNinjaLoading/MF_UI_DragonNinjaShuriken
graph {nodes [
  n1 = Constant2Vector() {g=256, r=256}
  n2 = FunctionInput(Preview=Constant() {r=0.2}) {bUsePreviewValueAsDefault=true, inputName=Size, inputType=FunctionInput_Scalar, sortPriority=1}
  n3 = Multiply(
   A=FunctionInput(Preview=Constant2Vector() {g=-0.3899, r=0.5}) {bUsePreviewValueAsDefault=true, inputName=Offset, inputType=FunctionInput_Vector2}
   B=@n2
  )
  n4 = Divide(
   A=Add(
    A=MaterialFunctionCall(Texture Dimensions=@n1) {materialFunction=&ScreenAlignedUVs}."X Scale to Ratio, Y 100%"
    B=@n3
   )
   B=AppendVector(A=@n2, B=@n2)
  )
  n5 = Subtract(
   A=Multiply(A=OneMinus(None=@n2)) {constB=30}
   B=MaterialFunctionCall(Float2=@n3) {materialFunction=&BreakOutFloat2Components}.R
  )
  n6 = Fmod(A=Multiply(A=Time(), B=Multiply(A=@n2) {constB=0.8}) {constB=0.5}, B=Constant() {r=1})
  n7 = AppendVector(A=LinearInterpolate(B=Multiply(A=@n5) {constB=-1}, Alpha=@n6) {constA=8, constB=-8}, B=Constant())
  n8 = OneMinus(None=@n6)
  n9 = TextureSample(UVs=MaterialFunctionCall(UVs=Add(A=@n4, B=@n7), "Rotation Angle (0-1)"=Multiply(A=@n8) {constB=6}) {materialFunction=&CustomRotator}.Rotated Values) {
   samplerSource=SSM_Clamp_WorldGroupSettings
   texture=&/T_UI_DragonNinja_Shuriken
  }
  n10 = ScalarParameter() {defaultValue=0.2, parameterName=Edge}
  n11 = Add(A=@n10, B=ScalarParameter() {defaultValue=0.021, parameterName=EdgeSoftness})
  n12 = Constant() {r=1}
  FunctionOutput(None=Multiply(
    A=MaterialFunctionCall(Alpha=@n9.R, Min=@n10, Max=@n11) {materialFunction=&SmoothStep}.Result
    B=FunctionInput(Preview=@n12) {bUsePreviewValueAsDefault=true, inputName=Alpha, inputType=FunctionInput_Scalar, sortPriority=3}
   ))
 ]}
