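# UI material for the background of the Season 8 DragonNinja loading screen.
# A screen-aligned radial gradient samples a curve atlas row for the base color.
# Six parameter-driven shuriken and six beam functions are each merged with Max, with beams masked by a bottom-centered radial gradient.
# Shurikens lerp from neutral gray to white, beams are added, and the sum is overlay-blended onto the base, so empty areas leave it unchanged.
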Material /BRCosmetics/2dAssets/Loadingscreens/season8/DragonNinjaLoading/M_UI_LoadingDragonNinja_BG
graph {
 properties {materialDomain=MD_UI, pixelDepthOffsetMode=PDOM_Legacy, usedShadingModels=Default Lit}
 nodes [
  n1 = VectorParameter() {parameterName=BGGradientRatio, defaultValue=<rgba 1920 256 0 1>}
  AppendVector(A=@n1.R, B=@n1.R)
  Constant()
  n2 = CurveAtlasRowParameter(CurveTime=MaterialFunctionCall(
    UVs=MaterialFunctionCall() {materialFunction=&ScreenAlignedUVs}."X Scale to Ratio, Y 100%"
    Radius=ScalarParameter() {defaultValue=0.5, parameterName=Radius}
    Density=ScalarParameter() {defaultValue=2.3, parameterName=Density}
   ) {materialFunction=&RadialGradientExponential}.RadialGradientExponential) {
   atlas=&/CurveLinearColorAtlas_Loading
   curve=&/Curve_UI_DragonNinjaBG
   parameterName=CurveParams
  }
  n3 = Constant3Vector() {constant=<rgba 0.5 0.5 0.5 0>}
  n4 = Constant3Vector() {constant=<rgba 1 1 1 0>}
  n5 = VectorParameter() {parameterName=Star1Params, defaultValue=<rgba 0 0 0 1>}
  n6 = AppendVector(A=@n5.R, B=@n5.R)
  n7 = VectorParameter() {parameterName=Star2Params, defaultValue=<rgba 0 0 0 1>}
  n8 = AppendVector(A=@n7.R, B=@n7.R)
  n9 = VectorParameter() {parameterName=Star3Params, defaultValue=<rgba 0 0 0 1>}
  n10 = Max(
   A=Max(
    A=MaterialFunctionCall!(Offset=@n6, Size=@n5.B, Alpha=@n5.B) {materialFunction=&/MF_UI_DragonNinjaShuriken}.Result
    B=MaterialFunctionCall(Offset=@n8, Size=@n7.B, Alpha=@n7.B) {materialFunction=&MF_UI_DragonNinjaShuriken}.Result
   )
   B=MaterialFunctionCall(Offset=AppendVector(A=@n9.R, B=@n9.R), Size=@n9.B, Alpha=@n9.B) {materialFunction=&MF_UI_DragonNinjaShuriken}.Result
  )
  n11 = VectorParameter() {parameterName=Star4Params, defaultValue=<rgba 0 0 0 1>}
  n12 = MaterialFunctionCall(Offset=AppendVector(A=@n11.R, B=@n11.R), Size=@n11.B, Alpha=@n11.B) {materialFunction=&MF_UI_DragonNinjaShuriken}
  n13 = VectorParameter() {parameterName=Star5Params, defaultValue=<rgba 0 0 0 1>}
  n14 = AppendVector(A=@n13.R, B=@n13.R)
  n15 = VectorParameter() {parameterName=Star6Params, defaultValue=<rgba 0 0 0 1>}
  n16 = Max(
   A=Max(
    A=Max(A=@n10, B=@n12.Result)
    B=MaterialFunctionCall(Offset=@n14, Size=@n13.B, Alpha=@n13.B) {materialFunction=&MF_UI_DragonNinjaShuriken}.Result
   )
   B=MaterialFunctionCall(Offset=AppendVector(A=@n15.R, B=@n15.R), Size=@n15.B, Alpha=@n15.B) {materialFunction=&MF_UI_DragonNinjaShuriken}.Result
  )
  n17 = VectorParameter() {parameterName=Beam1}
  n18 = MaterialFunctionCall(X=@n17.R, Y=@n17.R, Z=@n17.R, A=@n17.R) {materialFunction=&MakeFloat4}
  n19 = VectorParameter() {parameterName=Beam2}
  n20 = MaterialFunctionCall(X=@n19.R, Y=@n19.R, Z=@n19.R, A=@n19.R) {materialFunction=&MakeFloat4}
  n21 = VectorParameter() {parameterName=Beam3}
  n22 = Max(
   A=Max(
    A=MaterialFunctionCall!(In=@n18.Result) {materialFunction=&/MF_UI_DragonBeam}.Result
    B=MaterialFunctionCall(In=@n20.Result) {materialFunction=&MF_UI_DragonBeam}.Result
   )
   B=MaterialFunctionCall(In=MaterialFunctionCall(X=@n21.R, Y=@n21.R, Z=@n21.R, A=@n21.R) {materialFunction=&MakeFloat4}.Result) {materialFunction=&MF_UI_DragonBeam}.Result
  )
  n23 = VectorParameter() {parameterName=Beam4}
  n24 = MaterialFunctionCall(In=MaterialFunctionCall(X=@n23.R, Y=@n23.R, Z=@n23.R, A=@n23.R) {materialFunction=&MakeFloat4}.Result) {materialFunction=&MF_UI_DragonBeam}
  n25 = VectorParameter() {parameterName=Beam5}
  n26 = MaterialFunctionCall(X=@n25.R, Y=@n25.R, Z=@n25.R, A=@n25.R) {materialFunction=&MakeFloat4}
  n27 = VectorParameter() {parameterName=Beam6}
  n28 = Max(
   A=Max(
    A=Max(A=@n22, B=@n24.Result)
    B=MaterialFunctionCall(In=@n26.Result) {materialFunction=&MF_UI_DragonBeam}.Result
   )
   B=MaterialFunctionCall(In=MaterialFunctionCall(X=@n27.R, Y=@n27.R, Z=@n27.R, A=@n27.R) {materialFunction=&MakeFloat4}.Result) {materialFunction=&MF_UI_DragonBeam}.Result
  )
  n29 = MaterialFunctionCall(
   UVs=MaterialFunctionCall() {materialFunction=&ScreenAlignedUVs}."X Scale to Ratio, Y 100%"
   CenterPosition=Constant2Vector() {g=1, r=0.5}
   Radius=ScalarParameter() {defaultValue=0.5, parameterName=BeamRadius}
   Density=ScalarParameter() {defaultValue=2.3, parameterName=BeamDensity}
  ) {materialFunction=&RadialGradientExponential}
  MaterialFunctionCall(Base=@n2, Blend=Add(A=LinearInterpolate(A=@n3, B=@n4, Alpha=@n16), B=Multiply(A=@n28, B=@n29.RadialGradientExponential))) {materialFunction=&Blend_Overlay}
  TextureCoordinate()
 ]
}
